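; BPI_HuskDance is a Blueprint interface for AI, named for Husk dance behavior.
; It declares two functions, OnBeginDance and OnEndDance, that mark the start and end of a dance.
; Neither takes parameters or returns values, and the interface holds no logic, leaving the response to implementing classes.
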(fn OnBeginDance)

(fn OnEndDance)
